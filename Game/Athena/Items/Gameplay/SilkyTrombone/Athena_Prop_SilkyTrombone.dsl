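; Destructible Athena prop (SilkyTrombone) that dies on any damage and bursts, healing nearby pawns and applying water.
; If death effects are allowed, the server runs a scale-adjusted capsule overlap, applying Curie water to hit actors and a healing GameplayEffect to pawns.
; Death cosmetics spawn explosion VFX and a randomly rotated fading decal, then clear the damage emitter to prevent FX spam.
; The idle decal is randomly rotated at construction and removed on mobile.
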(event EventBeginPlay
  (|CallParentFunction)
  (if (TouchPlatform|IsMobileGame)
    (Components|DestroyComponent (Variables|Default|GetIdle_Decal))))

(event EventOnDeathServer
  (params (Damage :type Float%20%28single-precision%29) (DamageTags :type Gameplay%20Tag%20Container%20Structure%20%28by%20ref%29) (Momentum :type Vector :default "0, 0, 0") (HitInfo :type Hit%20Result%20Structure%20%28by%20ref%29) (InstigatedBy :type Controller%20Object%20Reference) (DamageCauser :type Actor%20Object%20Reference) (EffectContext :type Gameplay%20Effect%20Context%20Handle%20Structure))
  (if (Class|AthenaFunctionLibrary|ShouldAllowDeathEffects DamageTags)
    (CallFunction|DeactivateEmitter)
    (Variables|Default|SetWaterInstigator InstigatedBy)
    (bind (_outactors _returnvalue) (Collision|CapsuleOverlapActors (+ (Transformation|GetActorLocation) (Math|Vector|MakeVector 0.0 0.0 (Variables|Default|GetTowerHalfHeight))) (* (Variables|Default|GetExplosion%20Radius) (Math|Vector|VectorLength (Transformation|GetActorScale3D))) (* (Variables|Default|GetCapsuleHalfHeight) (.z (Transformation|GetActorScale3D))) (Variables|Default|GetOverlap%20Object%20Types)))
    (if _returnvalue
      (for _array_element _outactors
        (Utilities|IsValid _array_element
          (:Is%20Valid
            (bind _output_get (Variables|Default|SetHitActor _array_element))
            (CallFunction|ApplyWaterifCurieEnabled (Variables|Default|GetWaterInstigator) _output_get)
            (bind _asfort_pawn (Utilities|Casting|CastToFortPawn (Variables|Default|GetHitActor)))
            (Utilities|IsValid _asfort_pawn
              (:Is%20Valid
                (GameplayEffects|FortApplyGameplayEffectToTarget :self (Ability|GetFortAbilitySystemComponent self) :GameplayEffectClass (Variables|Default|GetGE_HealingEffect) :TargetOfEffect (Ability|GetAbilitySystemComponent _asfort_pawn) :OptionalSourceObject self :OptionalEffectCauser self :Level 0.0)))))))))

(event EventOnDeathPlayEffects
  (params (Damage :type Float%20%28single-precision%29) (DamageTags :type Gameplay%20Tag%20Container%20Structure%20%28by%20ref%29) (Momentum :type Vector :default "0, 0, 0") (HitInfo :type Hit%20Result%20Structure%20%28by%20ref%29) (InstigatedBy :type Fort%20Pawn%20Object%20Reference) (DamageCauser :type Actor%20Object%20Reference) (EffectContext :type Gameplay%20Effect%20Context%20Handle%20Structure))
  (bind _returnvalue (Effects|SpawnEmitteratLocation :EmitterTemplate (Variables|Default|GetExplosionVFX) :Location (Transformation|GetSocketLocation (Variables|Debug|GetStaticMeshComponent) "DeathFX") :Scale "1.000000,1.000000,1.000000" :bAutoDestroy true :PoolingMethod "AutoRelease" :bAutoActivateSystem true))
  (Variables|/Script/Engine.SceneComponent|SetbAbsoluteScale true _returnvalue)
  (bind _returnvalue_2 (Rendering|Decal|SpawnDecalatLocation "/Game/Athena/Items/Consumables/ChillBronco/FX/MI_Decal_ChillBronco_Hit_NoCuasticMotion.MI_Decal_ChillBronco_Hit_NoCuasticMotion" (Math|Vector|MakeVector (* (.x (Variables|Default|GetDeath%20Decal%20Size)) (.x (Transformation|GetActorScale3D))) (* (.y (Variables|Default|GetDeath%20Decal%20Size)) (.y (Transformation|GetActorScale3D))) (* (.z (Variables|Default|GetDeath%20Decal%20Size)) (.z (Transformation|GetActorScale3D)))) (Transformation|GetActorLocation) (Math|Rotator|MakeRotator (Math|Random|RandomFloatinRangefromStream (Variables|Default|GetRandom%20Stream) 0.0 359.0) -9e+01)))
  (Rendering|Components|Decal|SetFadeIn _returnvalue_2 0.0 0.0)
  (Rendering|Components|Decal|SetFadeOut _returnvalue_2 0.25 1.75 true)
  ;[0 0 0] Set FX to null, fx were being spammed by Athena_prop_burst_reaction -> FireEffect
  (Variables|Default|SetEmitterOnDamage 0))

(fn ApplyWaterIfCurieEnabled
  (params (Instigator :type Actor%20Object%20Reference) (Target :type Actor%20Object%20Reference))
  (Networking|CanExecuteCosmeticEvents
    (:False
      (if (Curie|IsCurieActive)
        (Utilities|IsValid Target
          (:Is%20Valid
            (Curie|ExecuteWaterApplication Instigator self Target 1e+02 (Collision|MakeHitResult false true 0.0 0.0 (Transformation|GetActorLocation) (Transformation|GetActorLocation) "0,0,1" "0,0,1" 0 Target) (Variables|Default|GetLandscapeExtinguishRadius))))))))

(fn ShouldDie
  (params (Damage :type Float%20%28single-precision%29) (EventInstigator :type Controller%20Object%20Reference) (DamageCauser :type Actor%20Object%20Reference) (bIsTest :type Boolean) (EffectContext :type Gameplay%20Effect%20Context%20Handle%20Structure%20%28by%20ref%29))
  (return true))

(fn ConstructionScript
  (|CallParentFunction)
  (Transformation|SetRelativeRotation (Variables|Default|GetIdle_Decal) (Math|Rotator|MakeRotator (Math|Random|RandomFloatinRangefromStream (Variables|Default|GetRandom%20Stream) 0.0 359.0) -9e+01) false false))
